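# Additive, unlit, two-sided particle material for glitch wisps surrounding the Haunt/Viper backpack.
# Turbulence UVs are radially distorted, rotated and panned by particle-driven timing, then sampled twice at an offset for a two-color chromatic split.
# Normal-map shading toward the actor center, time-of-day scaling, particle color and an inner hole shape the glow, with an optional curve-atlas ramp.
# Alpha combines a soft circular mask, depth fade and non-mobile view-angle fade.
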Material /BRCosmetics/Accessories/FORT_Backpacks/Backpack_M_MED_Viper/FX/Viper_Glitch/M_Backpack_Haunt_SurroundWisps
graph {
 properties {
  bUsedWithMeshParticles=true
  bUsedWithNiagaraMeshParticles=true
  bUsedWithNiagaraSprites=true
  blendMode=BLEND_Additive
  pixelDepthOffsetMode=PDOM_Legacy
  shadingModel=MSM_Unlit
  twoSided=true
  usedShadingModels=Unlit
 }
 nodes [
  n1 = ParticleColor()
  n2 = Constant() {r=1}
  n3 = MaterialFunctionCall(Value_Morning=@n2, Value_Day=ScalarParameter() {defaultValue=1.5, parameterName=DayMult}, Value_Evening=@n2, Value_Night=@n2) {materialFunction=&Scalar_Time_of_Day}
  n4 = ScalarParameter() {defaultValue=0.15, parameterName=Lerp min count}
  n5 = Subtract(A=ActorPositionWS(), B=WorldPosition().XYZ)
  n6 = ObjectRadius()
  n7 = SphereMask(
   A=TextureCoordinate()
   B=Constant() {r=0.5}
   Radius=ScalarParameter() {defaultValue=1, parameterName=OuterHoleRadius}
   Hardness=ScalarParameter() {defaultValue=0.15, parameterName=OuterHoleHardness}
  )
  n8 = ScalarParameter() {defaultValue=5, parameterName=DistortionPower}
  n9 = Multiply(A=Subtract(A=TextureCoordinate()) {constB=0.5}) {constB=2}
  n10 = Multiply(A=Multiply(A=Power(Base=@n7, Exp=@n8), B=@n9), B=ScalarParameter() {defaultValue=-0.35, parameterName=Distortion})
  n11 = TextureCoordinate() {uTiling=0.5, vTiling=0.5}
  n12 = DynamicParameter() {paramNames [Random,Speed,Param3,Param4]}
  n13 = ScalarParameter() {parameterName=Random Offset}
  n14 = Multiply(
   A=Add(A=Time(), B=Add(A=@n12.Random, B=@n13) {constB=0})
   B=Multiply(A=@n12.Speed, B=ScalarParameter() {defaultValue=2, parameterName=Speed})
  )
  n15 = Panner(Coordinate=Rotator(Coordinate=Add(A=@n10, B=@n11), Time=@n14) {centerX=0.2, centerY=0.2, speed=-0.07}, Time=@n14) {speedX=0.0004, speedY=-0.0002}
  n16 = TextureSample(UVs=@n15) {samplerType=SAMPLERTYPE_Normal, texture=/Game/Effects/Fort_Effects/Textures/Plasma/T_Turbulence_Seamless_N}
  n17 = Saturate(None=DotProduct(A=Divide(A=@n5, B=@n6), B=Transform(None=@n16.RGB)))
  n18 = Add(A=Multiply(A=ScalarParameter() {defaultValue=0.007, parameterName=ChromaOffset}, B=@n9) {constB=0.05}, B=@n15)
  n19 = LinearInterpolate(
   A=Multiply(
    A=TextureSample(UVs=@n18) {texture=/Game/Global/Textures/Default/Caustics/T_Turbulence_Seamless}.G
    B=VectorParameter() {parameterName=Chroma2, defaultValue=<rgba 1 0.0477 0.5272 1>}.RGB
   )
   B=Multiply(
    A=TextureSample(UVs=@n15) {texture=&T_Turbulence_Seamless}.G
    B=VectorParameter() {parameterName=Chroma1, defaultValue=<rgba 0.0296 1.3506 2 1>}.RGB
   )
  )
  n20 = Desaturation(
   None=Multiply(A=LinearInterpolate(A=@n4, Alpha=@n17) {constA=0.15, constB=2}, B=Power(Base=@n19) {constExponent=1.4})
   Fraction=ScalarParameter() {defaultValue=-0.75, parameterName=Desaturation}
  )
  n21 = ShadingPathSwitch(
   Default=MaterialFunctionCall(LDR Input=Saturate(None=@n20), HDR Tint=@n20, HDR Intensity Multiplier=ScalarParameter() {defaultValue=2, parameterName=HDRMult}) {materialFunction=&DeriveHDRfromLDR}.RemappedRange
   Mobile=@n20
  ) {desc=Migrated from FeatureLevelSwitch}
  n22 = SphereMask(
   A=TextureCoordinate()
   B=Constant() {r=0.5}
   Radius=ScalarParameter() {defaultValue=0.15, parameterName=InnerHoleRadius}
   Hardness=ScalarParameter() {defaultValue=0.7, parameterName=InnerHoleHardness}
  )
  n23 = Multiply(A=@n1.RGB, B=LinearInterpolate(A=Multiply(A=@n3.Result, B=@n21), Alpha=Power(Base=@n22) {constExponent=4}) {constB=0})
  n24 = CurveAtlasRowParameter(CurveTime=ComponentMask(None=RgbToHsv!(None=@n23)) {b=true}) {
   atlas=/Game/Accessories/FORT_Backpacks/Curves/CA_Backpack_Curves
   curve=/Game/Accessories/FORT_Backpacks/Curves/C_Backpack_HauntKoi_Ramp
   parameterName=Curve Atlas
  }
  MaterialFunctionCall(Emissive=LinearInterpolate(A=@n23, B=@n24, Alpha=ScalarParameter() {parameterName=Use Color Ramp}) {constAlpha=0}) {materialFunction=&MF_FX_EmissiveAdjustments}
  n25 = ScalarParameter() {defaultValue=1.5, parameterName=AngleFade}
  n26 = Saturate(None=Multiply(A=MaterialFunctionCall("FadeSensitivity (optional)"=@n25) {materialFunction=&FadeBasedOnViewAngle}.Result) {constB=5})
  n27 = TextureCoordinate()
  n28 = Constant() {r=0.5}
  n29 = DepthFade(
   Opacity=Power(Base=SphereMask(A=@n27, B=@n28, Radius=@n28) {hardnessPercent=70}) {constExponent=2.5}
   FadeDistance=ScalarParameter() {defaultValue=4, parameterName=FadeDistance}
  ) {fadeDistanceDefault=4}
  Multiply(A=@n1.A, B=ShadingPathSwitch(Default=Multiply(A=@n26, B=@n29), Mobile=@n29) {desc=Migrated from FeatureLevelSwitch})
 ]
}
